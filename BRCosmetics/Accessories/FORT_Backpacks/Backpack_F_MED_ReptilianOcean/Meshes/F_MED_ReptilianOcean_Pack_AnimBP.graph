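# Animation Blueprint for the ReptilianOcean backpack cosmetic, procedurally repositioning the pack to follow the wearer's body and actions.
# From the reference pose, chained additive Modify Bone nodes offset shield center, root and base pack.
# The event graph maps owner spine yaw to shield offsets and turns melee and forward-skydive-without-parachute states into interpolated alphas for extra tilts.
# Skydiving and weapon-derived firing-position values are set but unused by the pose.
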AnimBlueprint /BRCosmetics/Accessories/FORT_Backpacks/Backpack_F_MED_ReptilianOcean/Meshes/F_MED_ReptilianOcean_Pack_AnimBP
target_skeleton=/Game/Accessories/FORT_Backpacks/Backpack_F_MED_ReptilianOcean/Meshes/F_MED_ReptilianOcean_Pack_Skeleton
graphs [
 AnimGraph {nodes [
   n1 = "Transform (Modify) Bone - Bone: sheild_center"(
    ComponentPose=Local To Component(LocalPose=Local Space Ref Pose())
    Alpha=1
    Translation=Get backpack_translation()
    Rotation=Get backpack_rotate()
    Scale="1,1,1"
   ) {
    rotationMode=BMM_Additive
    rotationSpace=BCS_BoneSpace
    translationMode=BMM_Additive
    translationSpace=BCS_BoneSpace
    alphaBoolBlend {blendOption=Linear}
    alphaScaleBiasClamp {bInterpResult=true}
    boneToModify {boneName=sheild_center}
   }
   n2 = "Transform (Modify) Bone - Bone: root"(ComponentPose=@n1, Translation="0,4.5,2.5", Rotation="0,0,0") {
    translationMode=BMM_Additive
    translationSpace=BCS_BoneSpace
    alphaBoolBlend {blendOption=Linear}
    boneToModify {boneName=root}
    translation {y=4.5, z=2.5}
   }
   n3 = "Transform (Modify) Bone - Bone: base_pack"(ComponentPose=@n2, Alpha=Get Pickaxe() -> (Pickaxe:double), Translation="0,-2.5,3", Rotation="0,-8,10", Scale="1,1,1") {
    rotationMode=BMM_Additive
    rotationSpace=BCS_BoneSpace
    translationMode=BMM_Additive
    translationSpace=BCS_BoneSpace
    alphaBoolBlend {blendOption=Linear}
    alphaScaleBiasClamp {bInterpResult=true, interpSpeedDecreasing=2, interpSpeedIncreasing=20}
    boneToModify {boneName=base_pack}
    rotation {roll=10, yaw=-8}
    translation {y=-2.5, z=3}
   }
   n4 = "Transform (Modify) Bone - Bone: root"(ComponentPose=@n3, Alpha=Get Forward() -> (Forward:double), Translation="0,-5,0", Rotation="0,0,20") {
    rotationMode=BMM_Additive
    rotationSpace=BCS_BoneSpace
    translationMode=BMM_Additive
    translationSpace=BCS_BoneSpace
    alphaBoolBlend {blendOption=Linear}
    alphaScaleBiasClamp {bInterpResult=true, interpSpeedDecreasing=2, interpSpeedIncreasing=20}
    boneToModify {boneName=root}
    rotation {roll=20}
    translation {y=-5}
   }
   Output Pose(Result=Component To Local(ComponentPose=@n4)) {layerGroup=None}
  ]}
 EventGraph {nodes [
   Event BlueprintUpdateAnimation()
   > n1 = Sequence!(execute) -> (then_0, then_1, then_2:Exec)
   n2 = Cast To FortPlayerAnimInstance(execute=@n1.then_1, Object=GetAnimInstance(self=GetOwnerSkeletalMeshComponent()))
   > Set ar_firingPosition_GC(ar_firingPosition_GC:int=0) -> (Output_Get:int=0)
   > Switch on EFortWeaponCoreAnimation!(execute, Selection:EFortWeaponCoreAnimation Enum=Get WeaponCoreAnim(self=@n2)) -> (
    Melee:Exec
    Pistol:Exec
    Shotgun:Exec
    PaperBlueprint:Exec
    Rifle:Exec
    MeleeOneHand:Exec
    MachinePistol:Exec
    RocketLauncher:Exec
    GrenadeLauncher:Exec
    GoingCommando:Exec
    AssaultRifle:Exec
    TacticalShotgun:Exec
    SniperRifle:Exec
    TrapPlacement:Exec
    ShoulderLauncher:Exec
    AbilityDecoTool:Exec
    Crossbow:Exec
    C4:Exec
    RemoteControl:Exec
    DualWield:Exec
    AR_BullPup:Exec
    AR_ForwardGrip:Exec
    MedPackPaddles:Exec
    SMG_P90:Exec
    AR_DrumGun:Exec
    Consumable_Small:Exec
    Consumable_Large:Exec
    Balloon:Exec
    MountedTurret:Exec
    CreativeTool:Exec
    ExplosiveBow:Exec
    AshtonIndigo:Exec
    AshtonChicago:Exec
    MeleeDualWield:Exec
    Unarmed:Exec
   )
   > n3 = Branch()
   > n4 = Set ar_firingPosition_GC(ar_firingPosition_GC=1) -> (Output_Get=0)
   Set ar_firingPosition_GC(execute=@n3.else, ar_firingPosition_GC=0) -> (Output_Get=0)
   n5 = GetOwnerSkeletalMeshComponent()
   n6 = BreakRotator(InRot=BreakTransform(InTransform=GetDeltaTransformFromRefPose(self=@n5, BoneName=spine_05, BaseName=spine_01)).Rotation)
   Set backpack_rotate(execute=@n1.then_0, backpack_rotate=MakeRotator(Roll=MapRangeClamped(Value=@n6.Yaw, InRangeA=0, InRangeB=-30, OutRangeB=10)))
   > Set backpack_translation(backpack_translation=MakeVector(Y=MapRangeClamped(Value=@n6.Yaw, InRangeA=0, InRangeB=-30, OutRangeB=-8)))
   > Set Pickaxe(Pickaxe:double="To Float (Boolean)"(InBool=Get bIsPlayingMeleeAnim())) -> (Output_Get:double)
   > Set IsSkydiving(IsSkydiving:double="To Float (Boolean)"(InBool=Get bIsSkyDiving())) -> (Output_Get:double)
   n7 = PlayerMovement(execute=@n1.then_2, AnimInstance=Self-Reference(), XDotProductCoefficient=100, YDotProductCoefficient=100, ZDotProductCoefficient=100)
   n8 = NOT Boolean(A=Get bIsParachuteOpened())
   n9 = "float > float"(A=BreakVector(InVec=MakeVector(X=@n7, Y=@n7.YMovement)).Y, B=40)
   n10 = Get bIsSkyDiving()
   > Set Forward(Forward:double="To Float (Boolean)"(InBool=AND Boolean!(A=@n9, B=@n10, C:bool=@n8) -> (ReturnValue))) -> (Output_Get:double)
   n11 = GetOwnerSkeletalMeshComponent()
   n12 = BreakRotator(InRot=BreakTransform(InTransform=GetDeltaTransformFromRefPose(self=@n11, BoneName=spine_05, BaseName=spine_01)).Rotation)
   n13 = MapRangeClamped(Value=@n12.Yaw, InRangeA=0, InRangeB=-30, OutRangeB=35)
   Set backpack_rotate(backpack_rotate=MakeRotator(Roll=FInterpEaseInOut!(A:double=@n13, B:double=0, Alpha:double=0, Exponent:double=0) -> (ReturnValue:double)))
   n14 = MapRangeClamped(Value=@n12.Yaw, InRangeA=0, InRangeB=-30, OutRangeB=-10)
   > Set backpack_translation(backpack_translation=MakeVector(Y=FInterpEaseInOut(A=@n14, B=0, Alpha=0, Exponent=0)))
   TryGetPawnOwner()
   "To Float (Integer)"(InInt=@n4)
  ]}
]
variables [
 ar_firingPosition_GC {type=integer}
 backpack_rotate {type=Rotator}
 backpack_rotate_0 {type=Rotator}
 backpack_translation {type=Vector}
 Forward {type=number}
 IsSkydiving {type=number}
 Pickaxe {type=number}
]
